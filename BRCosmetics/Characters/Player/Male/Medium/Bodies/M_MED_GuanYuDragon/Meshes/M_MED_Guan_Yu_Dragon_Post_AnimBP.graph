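# Post-process AnimBlueprint for the male medium Guan Yu Dragon body, applying small bone corrections to the incoming pose.
# AnimGraph pins the dyn_noseRing bone at a fixed parent-space offset, then additively scales upperarm_twist_01_r with armSwingDelta_r as alpha.
# Each update derives that alpha from the absolute yaw of the right upper arm's ref-pose delta against the clavicle, using a fixed value while skydiving.
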AnimBlueprint /BRCosmetics/Characters/Player/Male/Medium/Bodies/M_MED_GuanYuDragon/Meshes/M_MED_Guan_Yu_Dragon_Post_AnimBP
target_skeleton=/BRCosmetics/Characters/Player/Male/Medium/Bodies/M_MED_GuanYu_01/Export/M_MED_Commando_Guan_Yu_Skeleton
graphs [
 AnimGraph {nodes [
   n1 = "Transform (Modify) Bone - Bone: dyn_noseRing"(ComponentPose=Local To Component(LocalPose=Input Pose()), Alpha=1, Translation="7.538,-0.071,-4.23", Rotation="0,0,0", Scale="1,1,1") {
    translationMode=BMM_Replace
    translationSpace=BCS_ParentBoneSpace
    alphaBoolBlend {blendOption=Linear}
    boneToModify {boneName=dyn_noseRing}
    translation=<7.538 -0.071 -4.23>
   }
   n2 = "Transform (Modify) Bone - Bone: upperarm_twist_01_r"(ComponentPose=@n1, Alpha=Get armSwingDelta_r() -> (armSwingDelta_r:double), Translation="0,0,0", Rotation="0,0,0", Scale="1.5,1.5,1.5") {scaleMode=BMM_Additive, alphaBoolBlend {blendOption=Linear}, boneToModify {boneName=upperarm_twist_01_r}, scale=<1.5 1.5 1.5>}
   Output Pose(Result=Component To Local(ComponentPose=@n2)) {layerGroup=None}
  ]}
 EventGraph {nodes [
   Event BlueprintUpdateAnimation()
   n1 = GetDeltaTransformFromRefPose(self=GetOwnerSkeletalMeshComponent(), BoneName=upperarm_r, BaseName=clavicle_r)
   n2 = "Absolute (Float)"(A=BreakRotator(InRot=BreakTransform(InTransform=@n1).Rotation).Yaw)
   > Set armSwingDelta_r(armSwingDelta_r:double=SelectFloat(A=0.5, B=MapRangeClamped(Value=@n2, InRangeA=90, InRangeB=130, OutRangeB=0.75), bPickA=Get bIsSkyDiving())) -> (Output_Get:double)
   TryGetPawnOwner()
  ]}
]
variables [armSwingDelta_r {type=number}]
